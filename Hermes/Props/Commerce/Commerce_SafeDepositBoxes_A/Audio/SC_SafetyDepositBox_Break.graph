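# Sound cue for the break sound of a safety deposit box prop, played in the impacts sound class.
# A random node picks one of four equally weighted break recordings and feeds a modulator that varies playback slightly.
# The cue uses a dedicated safety deposit box attenuation asset and a boosted volume.
# A concurrency override caps simultaneous instances and prevents new ones once the limit is reached.
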SoundCue /Hermes/Props/Commerce/Commerce_SafeDepositBoxes_A/Audio/SC_SafetyDepositBox_Break
root=Modulator_1
nodes [
 {id=Modulator_1, children [Random_0], properties {pitchMin=1}}
 {id=Random_0, children [WavePlayer_0,WavePlayer_1,WavePlayer_2,WavePlayer_3], properties {weights [1,1,1,1]}}
 {id=WavePlayer_0, properties {soundWaveAssetPtr=&/SafetyDepositBox_Break_01}}
 {id=WavePlayer_1, properties {soundWaveAssetPtr=&/SafetyDepositBox_Break_02}}
 {id=WavePlayer_2, properties {soundWaveAssetPtr=&/SafetyDepositBox_Break_03}}
 {id=WavePlayer_3, properties {soundWaveAssetPtr=&/SafetyDepositBox_Break_04}}
]
properties {
 attenuationSettings=&/ATT_SafetyDepositBox
 bOverrideConcurrency=true
 maxDistance=6000
 soundClassObject=&Fort_Impacts
 volumeMultiplier=1.8
 attenuationOverrides {bAttenuate=false, falloffDistance=10000, attenuationShapeExtents {x=0}}
 concurrencyOverrides {maxCount=5, resolutionRule=PreventNew}
}
