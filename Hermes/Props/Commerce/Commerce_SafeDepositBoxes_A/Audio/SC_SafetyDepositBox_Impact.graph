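# SoundCue for impact sounds on the safety deposit box prop.
# A Random node picks one of four equally weighted impact waves and feeds it into a Modulator for per-play variation.
# It plays in the Fort_Impacts sound class with a slight volume boost and the prop's shared attenuation asset.
# A concurrency override caps simultaneous instances at five and prevents new ones beyond that.
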SoundCue /Hermes/Props/Commerce/Commerce_SafeDepositBoxes_A/Audio/SC_SafetyDepositBox_Impact
root=Modulator_0
nodes [
 {id=Modulator_0, children [Random_3]}
 {id=Random_3, children [WavePlayer_12, WavePlayer_13, WavePlayer_14, WavePlayer_15], properties {weights [1,1,1,1]}}
 {id=WavePlayer_12, properties {soundWaveAssetPtr=&/SafetyDepositBox_Impact_04}}
 {id=WavePlayer_13, properties {soundWaveAssetPtr=&/SafetyDepositBox_Impact_03}}
 {id=WavePlayer_14, properties {soundWaveAssetPtr=&/SafetyDepositBox_Impact_02}}
 {id=WavePlayer_15, properties {soundWaveAssetPtr=&/SafetyDepositBox_Impact_01}}
]
properties {
 attenuationSettings=&/ATT_SafetyDepositBox
 bOverrideConcurrency=true
 maxDistance=6000
 soundClassObject=&Fort_Impacts
 volumeMultiplier=1.2
 attenuationOverrides {bAttenuate=false, falloffDistance=10000, attenuationShapeExtents {x=0}}
 concurrencyOverrides {maxCount=5, resolutionRule=PreventNew}
}
