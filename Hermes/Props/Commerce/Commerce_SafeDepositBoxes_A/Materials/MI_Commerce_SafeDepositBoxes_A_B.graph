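# Simple Default Lit material for the Commerce safe deposit box prop, a base material despite its MI_ name, flagged for Nanite, instanced meshes and geometry collections.
# A dark yellow Color parameter feeds a RayTracingQualitySwitch, alongside a separate scalar Roughness parameter.
# The ray-traced branch scales that color by BankYellowLumenMult, which defaults to zero, apparently to control the yellow's contribution to Lumen.
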Material /Hermes/Props/Commerce/Commerce_SafeDepositBoxes_A/Materials/MI_Commerce_SafeDepositBoxes_A_B
graph {
 properties {
  bUsedWithGeometryCollections=true
  bUsedWithInstancedStaticMeshes=true
  bUsedWithNanite=true
  bUsedWithStaticLighting=true
  pixelDepthOffsetMode=PDOM_Legacy
  usedShadingModels=Default Lit
 }
 nodes [
  n1 = VectorParameter() {parameterName=Color, defaultValue=<rgba 0.25 0.212 0.0202 1>}
  RayTracingQualitySwitch(Normal=@n1.RGB, RayTraced=Multiply(A=@n1.RGB, B=ScalarParameter() {parameterName=BankYellowLumenMult}) {constB=0.25})
  ScalarParameter() {defaultValue=0.56, parameterName=Roughness}
 ]
}
