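; Parent Blueprint for breakable safe deposit box props that spawn loot when damaged.
; On authority, BeginPlay prefixes each playlist map value with the instance's search loot tier group, then selects the entry matching playlist context tags once playlist data is ready.
; Server damage accumulates to a threshold that sets Damaged, whose OnRep spawns loot, using the damaging player pawn when valid, and plays a break-open sound.
; Death plays a separate break sound.
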(event EventBeginPlay
  (|CallParentFunction)
  (Networking|SwitchHasAuthority
    (:Authority
      (CallFunction|SetupPlaylistMap)
      (Class|FortAsyncActionCurrentPlaylistReady|PlaylistDataReadyAsync
        (:OnReady
          (CallFunction|PickLootGroupbyPlaylist))))))

(event Collision|EventActorBeginOverlap
  (params (OtherActor :type Actor%20Object%20Reference))
  (|CallParentFunction OtherActor) ; This node is disabled and will not be called.\nDrag off pins to build functionality.
)

(event EventTick
  (params (DeltaSeconds :type Float%20%28single-precision%29))
  (|CallParentFunction DeltaSeconds) ; This node is disabled and will not be called.\nDrag off pins to build functionality.
)

(event EventOnDamageServer
  (params (Damage :type Float%20%28single-precision%29) (DamageTags :type Gameplay%20Tag%20Container%20Structure%20%28by%20ref%29) (Momentum :type Vector :default "0, 0, 0") (HitInfo :type Hit%20Result%20Structure%20%28by%20ref%29) (InstigatedBy :type Controller%20Object%20Reference) (DamageCauser :type Actor%20Object%20Reference) (EffectContext :type Gameplay%20Effect%20Context%20Handle%20Structure))
  (bind _output_get (Variables|Default|SetTotalDamage (+ (Variables|Default|GetTotalDamage) Damage)))
  (if (>= _output_get (Variables|Default|GetDamageThreshold))
    (Variables|Default|SetDamaged true)
    (bind _asfort_player_pawn (Utilities|Casting|CastToFortPlayerPawn DamageCauser))
    (Variables|Default|SetAs%20Fort%20Player%20Pawn _asfort_player_pawn)))

(event EventOnDeathPlayEffects
  (params (Damage :type Float%20%28single-precision%29) (DamageTags :type Gameplay%20Tag%20Container%20Structure%20%28by%20ref%29) (Momentum :type Vector :default "0, 0, 0") (HitInfo :type Hit%20Result%20Structure%20%28by%20ref%29) (InstigatedBy :type Fort%20Pawn%20Object%20Reference) (DamageCauser :type Actor%20Object%20Reference) (EffectContext :type Gameplay%20Effect%20Context%20Handle%20Structure))
  (|CallParentFunction Damage DamageTags Momentum HitInfo InstigatedBy DamageCauser EffectContext)
  (Audio|PlaySoundatLocation "/Hermes/Props/Commerce/Commerce_SafeDepositBoxes_A/Audio/SC_SafetyDepositBox_Break.SC_SafetyDepositBox_Break" (Transformation|GetWorldLocation (Variables|Debug|GetStaticMeshComponent))))

(fn OnRep_Damaged
  (Fortnite|SetCanbeDamaged self true)
  (bind _as_fort_player_pawn (Variables|Default|GetAs%20Fort%20Player%20Pawn)
    (:then
      (Container|SpawnLoot _as_fort_player_pawn))
    (:else
      (Container|SpawnLoot 0)))
  (Audio|PlaySoundatLocation (Variables|Default|GetBreak%20Open%20Sound) (Transformation|GetWorldLocation (Variables|Debug|GetStaticMeshComponent))))

(fn PickLootGroupByPlaylist
  (bind _keys (Utilities|Map|Keys (Variables|Default|GetLootGroupByPlaylistMap)))
  (bind (_array_element _array_index) (Utilities|Array|ForEachLoopwithBreak _keys)
    (:LoopBody
      (bind _returnvalue (Fortnite|GameplayTags|HasTagfromName (GameplayTags|GetTagName _array_element) (Class|FortGameStateAthena|GetAthenaPlaylistContextTags (Fortnite|Athena|GameState|GetGameStateAthena))))
      (if _returnvalue
        (bind _values (Utilities|Map|Values (Variables|Default|GetLootGroupByPlaylistMap)))
        (bind _output_get (Variables|Container|SetSearchLootTierGroup (Utilities|Array|Get(acopy) _values _array_index)))
        (Container|SetSearchLootTierGroup _output_get true true)))))

(fn SetupPlaylistMap
  ; Allows us to override the loot tier on a per instance basis
  (bind _keys (Utilities|Map|Keys (Variables|Default|GetLootGroupByPlaylistMap)))
  (for _array_element _keys
    (Utilities|Map|Add (Variables|Default|GetLootGroupByPlaylistMap) _array_element (Utilities|String|StringToName (Utilities|String|Append (Utilities|String|ToString(Name) (Variables|Container|GetSearchLootTierGroup)) (Utilities|String|ToString(Name) (Utilities|Map|Find (Variables|Default|GetLootGroupByPlaylistMap) _array_element)))))))

(fn ConstructionScript
  (|CallParentFunction))
